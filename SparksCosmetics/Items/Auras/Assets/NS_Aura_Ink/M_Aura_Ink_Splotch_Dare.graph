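# Translucent unlit Niagara particle material that draws animated ink splotches for the Ink aura cosmetic.
# Dynamic parameters drive sub-UV frame, scale and rotation, while panning radial cloud noise distorts the ink lookups.
# Particle color tints the ink, multiplied by per-particle flipbook brush and object-locked screen-space paper textures.
# Opacity uses a noise-perturbed radial reveal with depth fade except on mobile; erode branches are left unconnected.
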Material /SparksCosmetics/Items/Auras/Assets/NS_Aura_Ink/M_Aura_Ink_Splotch_Dare
graph {
 properties {
  bAutomaticallySetUsageInEditor=false
  bUsedWithNiagaraMeshParticles=true
  bUsedWithNiagaraSprites=true
  blendMode=BLEND_Translucent
  shadingModel=MSM_Unlit
  usedShadingModels=Unlit
 }
 nodes [
  n1 = MaterialFunctionCall!(Center Point World Position=ObjectPositionWS()) {
   desc=Locks screenspace effect to Object Position
   materialFunction=/Engine/Functions/Engine_MaterialFunctions02/UVs/WorldSpaceAlignedScreenCoordinates
  }
  n2 = Multiply(A=@n1.Offset and Scaled UVS, B=ScalarParameter() {defaultValue=8, parameterName=ScreenspaceTex_Scale})
  n3 = Constant() {r=2}
  n4 = TextureSample(UVs=MaterialFunctionCall("Animation  Phase (0-1)"=ParticleRandom(), Number of Rows=@n3, Number of Columns=@n3) {materialFunction=&FlipBook}.UVs) {texture=&/T_InkTextures_Dare}
  n5 = ScalarParameter() {defaultValue=0.8, parameterName=BrushTextureStrength}
  n6 = ParticleColor()
  n7 = DynamicParameter() {parameterIndex=1, paramNames [SubUV Frame, SubUV Frame Scale, SubUV Frame Rotation, Param4]}
  n8 = MaterialFunctionCall(
   UVs=MaterialFunctionCall(UVs=TextureCoordinate(), Rotation Center=Constant2Vector() {g=0.5, r=0.5}, "Rotation Angle (0-1)"=@n7.SubUV Frame Rotation) {materialFunction=&CustomRotator}.Rotated Values
   Texture Scale=AppendVector(A=@n7.SubUV Frame Scale, B=@n7.SubUV Frame Scale)
  ) {materialFunction=&ScaleUVsByCenter}
  n9 = MaterialFunctionCall() {materialFunction=&VectorToRadialValue}
  n10 = TextureCoordinate()
  n11 = TextureSample!(
   UVs=Panner(Coordinate=Multiply(A=@n9.Radial Coordinates)) {speedX=0.025, speedY=-0.05}
   Tex
   "DDX(UVs)"=DDX(Value=@n10)
   "DDY(UVs)"=DDY(Value=@n10)
   Apply View MipBias
  ) {
   mipValueMode=TMVM_Derivative
   samplerType=SAMPLERTYPE_LinearGrayscale
   texture=/Game/Effects/Fort_Effects/Emotes/UltraLight/Texture/T_Cloud_Greyscale
  }
  n12 = ScalarParameter() {defaultValue=3, parameterName=Mask - SubImage U}
  n13 = ScalarParameter() {defaultValue=3, parameterName=Mask - SubImage V}
  n14 = Add(
   A=Multiply(A=@n11.G, B=ScalarParameter() {defaultValue=0.02, parameterName=Distortion Strength}) {constB=0.02}
   B=MaterialFunctionCall!(UVs=@n8.UVs, SubImages=AppendVector(A=@n12, B=@n13), Frame=@n7.SubUV Frame) {materialFunction=/BRCosmetics/Animation/Game/MainPlayer/Emotes/TroutWrist/FX/Materials/SubUV_Function}.UVs
  )
  n15 = Multiply(
   A=@n8.0-1 mask
   B=ChannelMaskParameter(None=TextureSampleParameter2D(UVs=@n14) {parameterName=Texture, texture=&/T_Aura_Ink}.RGBA) {parameterName=Texture}.R
  )
  n16 = Multiply(
   A=LinearInterpolate(
    B=TextureSample!(UVs=@n2, Tex, Apply View MipBias) {texture=&T_InkTextures_Dare}.G
    Alpha=ScalarParameter() {defaultValue=0.75, parameterName=PaperTextureStrength}
   ) {constA=1}
   B=Multiply(A=LinearInterpolate(B=@n4.R, Alpha=@n5) {constA=1}, B=Multiply(A=@n6.RGB, B=@n15))
  )
  MaterialFunctionCall(Emissive=@n16) {materialFunction=&MF_FX_EmissiveAdjustments}
  n17 = OneMinus(None=DynamicParameter() {defaultValue=<rgba 0.9 -0.1 0.5 0.1>, paramNames [Reveal, Erode, RGE Radius, Reveal Edge Softness]}.Erode)
  n18 = ScalarParameter() {defaultValue=0.1, parameterName=Erode Edge Softness}
  n19 = SmoothStep() {constMax=1.5}
  Multiply(B=SmoothStep(Min=@n17, Max=Add(A=@n18, B=@n17) {constB=-0.1}, Value=OneMinus(None=@n19)) {constMax=0, constMin=1})
  Power(
   Base=LinearInterpolate(B=ScalarParameter() {defaultValue=0.5, parameterName=MaxTextureBrightness}, Alpha=@n16)
   Exp=ScalarParameter() {defaultValue=2, parameterName=TexturePow}
  )
  ScalarParameter() {defaultValue=0.1, parameterName=Reveal Edge Softness}
  n20 = TextureSampleParameter2D(UVs=@n14) {parameterName=Opacity, texture=&T_Aura_Ink}
  n21 = Power(
   Base=Multiply(A=Multiply(A=@n8.0-1 mask, B=@n20.G), B=ScalarParameter() {defaultValue=1, parameterName=Opacity Mult})
   Exp=ScalarParameter() {defaultValue=1, parameterName=Opacity Pow}
  )
  n22 = DynamicParameter() {defaultValue=<rgba 0.9 -0.1 0.5 0.1>, paramNames [Reveal, Erode, RGE Radius, Reveal Edge Softness]}
  n23 = OneMinus(None=@n22.Reveal)
  n24 = Add(A=@n22.Reveal Edge Softness, B=@n23) {constB=-0.1}
  n25 = Multiply(A=TextureSample(UVs=TextureCoordinate()) {texture=&AtmosphericCloudNoise01}.G) {constB=0.1}
  n26 = Multiply(
   A=Multiply(A=@n8.0-1 mask, B=ChannelMaskParameter(None=@n20.RGBA) {parameterName=Opacity}.R)
   B=MaterialFunctionCall(Radius=Add(A=@n22.RGE Radius, B=@n25) {constB=0}) {materialFunction=&RadialGradientExponential}.RadialGradientExponential
  )
  n27 = Multiply(A=@n6.A, B=Multiply(A=Saturate(None=@n21), B=SmoothStep(Min=@n23, Max=@n24, Value=@n26) {constMax=0, constMin=1}))
  n28 = DepthFade(Opacity=@n27, FadeDistance=ScalarParameter() {defaultValue=15, parameterName=Depth Fade Distance})
  ShadingPathSwitch(Default=@n28, Deferred=@n28, Mobile=@n27) {desc=Migrated from FeatureLevelSwitch}
  n29 = DynamicParameter() {defaultValue=<rgba 0.9 -0.1 0.5 0.1>, paramNames [Reveal, Erode, RGE Radius, Reveal Edge Softness]}
  SmoothStep(
   Min=@n29.Erode
   Max=Add(A=@n29.Erode, B=ScalarParameter() {defaultValue=0.1, parameterName=Erode Edge Softness})
   Value=ChannelMaskParameter(None=TextureSampleParameter2D(UVs=@n14) {parameterName=Opacity, texture=&T_Aura_Ink}.RGBA) {parameterName=Opacity}.R
  )
 ]
}
