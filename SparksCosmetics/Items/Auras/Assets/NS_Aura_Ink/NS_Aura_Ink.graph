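# Niagara system for the Ink cosmetic aura, layering bone-attached ink effects on a character's skeletal mesh.
# Glow sprites sit on bones, rising smoke spawns along two bones, and fire meshes pass smoothed positions for trailing.
# Butterfly meshes with ink-blot sprites drift on wind inside a volume constraint before dissolving, while animated splotches lag behind a bone with speed-scaled smoothing.
# User flags adjust brightness for front-end display and scale for large bodies.
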NiagaraSystem /SparksCosmetics/Items/Auras/Assets/NS_Aura_Ink/NS_Aura_Ink
dependencies {
 usedDataInterfaces [
  &NiagaraDataInterfaceArrayFloat
  &NiagaraDataInterfaceArrayInt32
  &NiagaraDataInterfaceArrayFloat3
  &NiagaraDataInterfaceCurve
  &NiagaraDataInterfaceSpriteRendererInfo
  &NiagaraDataInterfaceArrayFloat2
  &NiagaraDataInterfaceSkeletalMesh
 ]
 usedDynamicInputs []
 usedModules [
  &self:SystemSpawnScript.SystemScriptSource.SystemScriptGraph.NiagaraNodeAssignment_3.SetVariables_#1
  &SystemState
 ]
}
emitters {
 Butterflies {
  emitterUpdateScript [
   &EmitterState()
   &SpawnRate(SpawnRate=2, Use Spawn Probability=false)
  ]
  particleSpawnScript [
   &InitializeParticle(
    Color=&MakeLinearColorFromVectorAndFloat(
     "Vector (RGB)"=&MakeVectorFromLinearColorRGB(Color=&Multiply_LinearColorByFloat(
       Linear Color=<rgba 0.5052 0.5052 0.5052 1>
       Float=&MakeCustomFloatFromBool(Bool=$User.IsFrontEnd, True Float=1, False Float=1.5)
      ))
     "Float (Alpha)"=1
    )
    Lifetime=&RandomRangeFloat(Minimum=4, Maximum=4)
    Mesh Scale Mode=::NewEnumerator1
    Mesh Uniform Scale Max=0.025
    Mesh Uniform Scale Min=0.015
    Sprite Rotation Mode=::NewEnumerator2
    Sprite Size Mode=::NewEnumerator4
    UsePositionOffset=false
    Write Material Random=true
   )
   &SkeletalMeshLocation(
    Bone / Socket Index=7
    Bone Sampling Mode=::NewEnumerator4
    MaskSpawnGroup=false
    Orientation Sampling=::NewEnumerator1
    Skeletal Mesh=$System.Skeletal Mesh
   )
   &ShapeLocation(MaskSpawnGroup=false, Override Seed=false, Sphere Radius=10, UseOwnerScale=true) {module=ShapeLocation001}
   &ApplyInitialForces()
   /SetVariables(Particles.GlowDynamicParams:Vector4f=<0.5 1 0 0>) {isSetParameters=true}
   &CameraOffset(Camera Offset Amount=-40)
   /SetVariables(
    Particles.InkBlotColor:LinearColor=&MakeLinearColorFromVectorAndFloat(
     "Vector (RGB)"=&MakeVectorFromLinearColorRGB(Color=&Multiply_LinearColorByFloat(
       Linear Color=&RandomRangeLinearColor(Minimum=<rgba 0.0781 0.0781 0.0781 1>, Maximum=<rgba 0.0208 0.0208 0.0208 1>)
       Float=&MakeCustomFloatFromBool(Bool=$User.IsFrontEnd, True Float=1, False Float=1.5)
      ))
     "Float (Alpha)"=1
    )
    Particles.InkBlotDynamicParam1:Vector4f=&MakeVector4(
     X=&SelectFloatFromArray(
      Array Sampling Mode=::NewEnumerator0
      Direct Array Index=?
      Float Selection Array=DataInterface() {propertyValues {FloatData [0,1,2,4,7]}}
      Random Seed=0
     )
     Y=1
     Z=&FloatFromCurve(FloatCurve=DataInterface() {propertyValues {Curve {keys [<at 0 1>,<at 1 1>]}}})
     W=0
    )
    Particles.InkBlotSize:Vector2f=&Vector2DFromFloat(Value=&RandomRangeFloat(Minimum=10, Maximum=20))
   ) {isSetParameters=true}
  ]
  particleUpdateScript [
   &ParticleState()
   &DynamicMaterialParameters(
    Index 0 Param 1=0.25
    Index 0 Param 3=1
    Index 0 Param 4=&Multiply_Float(A=$Engine.Time, B=&RandomRangeFloat(Minimum=0.3, Maximum=0.4))
    Index 1 Param 1=3000
    Index 1 Param 2=1
    Module.Param0WriteEnabled=true
    Module.Param0WriteEnabled1=true
    Module.Param1WriteEnabled=true
    Module.Param1WriteEnabled1=true
    Module.Param2WriteEnabled=true
    Module.Param2WriteEnabled1=true
    Module.Param3WriteEnabled=true
    Module.Param3WriteEnabled1=true
    Write Parameter Index 1=true
   )
   &UpdateMeshOrientation(
    Facing Direction=&SubtractVectorFromPosition(
     Position=/Niagara/DynamicInputs/Helpers/SimulationPosition()
     Vector=Linked() {linkedVariable {name=Particles.Position, type=NiagaraPosition}}
    )
    Facing Interpolation=false
    Orientation Method=::NewEnumerator1
    Override Quat=false
    TransformFacing=false
    TransformUp=false
    Up Direction:Vector3f=<1 0 0>
   )
   &WindForce(
    Frequency=10
    Scale=1
    Turbulence Flow Loop Duration=0
    Wind Speed=&RandomRangeVector(Minimum=<-7 -7 4>, Maximum=<7 7 5>)
    Wind Speed Scale=&FloatFromCurve(
     FloatCurve=DataInterface() {propertyValues {Curve {keys [<at 0 1>, {arriveTangent=-1.4286, interpMode=RCIM_Cubic, leaveTangent=-1.4286, time=0.6, value=1}, <at 0.7 0 RCIM_Cubic>]}}}
     Scale Curve=2.5
    )
   )
   &AerodynamicDrag(
    Aerodynamic Drag:NiagaraFloat=&RandomRangeFloat(Minimum=0.3, Maximum=0.7)
    Aerodynamic Rotational Drag=0
    Drag Shape Curve=DataInterface() {propertyValues {Curve {keys [
        {interpMode=RCIM_Cubic, tangentMode=RCTM_User, time=0, value=0.1}
        {arriveTangent=2.0799, interpMode=RCIM_Cubic, leaveTangent=2.0799, tangentMode=RCTM_User, time=1, value=1}
       ]}}}
    Lift Contribution=0
    Lift Shape Curve=DataInterface() {propertyValues {Curve {keys [
        {interpMode=RCIM_Cubic, tangentMode=RCTM_User, time=0, value=0}
        {interpMode=RCIM_Cubic, tangentMode=RCTM_User, time=0.5, value=1}
        {tangentMode=RCTM_User, time=1, value=0}
       ]}}}
    Use Linear Drag=true
    Use Rotational Drag=true
   ) {module=AerodynamicDrag001}
   &ScaleMeshSize(Scale Factor=&VectorFromFloat(Value=&FloatFromCurve(FloatCurve=DataInterface() {propertyValues {Curve {keys [
          <at 0 0 RCIM_Cubic>
          {arriveTangent=1.1111, leaveTangent=1.1111, time=0.15, value=1}
          {arriveTangent=-1.8182, interpMode=RCIM_Cubic, leaveTangent=-1.8182, time=0.6, value=1}
          {interpMode=RCIM_Cubic, leaveTangent=-1.1765, time=0.7, value=0}
         ]}}})))
   &ScaleSpriteSize(Uniform Curve Sprite Scale=DataInterface() {propertyValues {Curve {keys [
        <at 0 0 RCIM_Cubic>
        {arriveTangent=1.1111, leaveTangent=1.1111, time=0.15, value=1}
        {arriveTangent=-1.8182, interpMode=RCIM_Cubic, leaveTangent=-1.8182, time=0.6, value=1}
        {interpMode=RCIM_Cubic, leaveTangent=-1.1765, time=0.7, value=0}
       ]}}})
   &self:Volume_Constraint(
    Max Distance=60
    Min Distance=20
    Non-uniform Scale:Vector3f=<1 1 1.1>
    VelocityMult:NiagaraFloat=&FloatFromCurve(
     FloatCurve=DataInterface() {propertyValues {Curve {keys [<at 0 1>, {arriveTangent=-1.4286, interpMode=RCIM_Cubic, leaveTangent=-1.4286, time=0.6, value=1}, <at 0.7 0 RCIM_Cubic>]}}}
     Scale Curve=2
    )
    Volume Center:NiagaraPosition=&AddVectorToPosition(
     Position=$Particles.SkeletalMeshLocation.SampledPosition
     Vector=&RotateVectorByQuaternion(Vector To Rotate=<-10 0 0>, Quaternion=$Particles.SkeletalMeshLocation.SampledRotation)
    )
   ) {module=Volume_Constraint}
   &SolveForcesAndVelocity(Clamp Velocity=false, Limit Acceleration=false)
   &SpriteFacingAndAlignment(Sprite Alignment=<0 0 1>, WriteAlignment=true, WriteFacing=false)
   &AlignSpriteToMeshOrientation(SetAlignment=true, SetFacing=true)
   /SetVariables(
    Particles.InkBlotColor:LinearColor=&MakeLinearColorFromVectorAndFloat(
     "Vector (RGB)"=&MakeVectorFromLinearColorRGB(Color=&RandomRangeLinearColor(Minimum=<rgba 0.0104 0.0104 0.0104 1>, Maximum=<rgba 0.0156 0.0156 0.0156 1>))
     "Float (Alpha)"=&FloatFromCurve(FloatCurve=DataInterface() {propertyValues {Curve {keys [<at 0.7 1 RCIM_Cubic>, {arriveTangent=-6.9791, interpMode=RCIM_Cubic, leaveTangent=-6.9791, tangentMode=RCTM_User, time=1, value=0}]}}})
    )
    Particles.InkBlotDynamicParam0:Vector4f=&MakeVector4(
     X=&FloatFromCurve(FloatCurve=DataInterface() {propertyValues {Curve {keys [{arriveTangent=5.3049, interpMode=RCIM_Cubic, leaveTangent=5.3049, tangentMode=RCTM_User, time=0.59, value=0}, <at 1 1 RCIM_Cubic>]}}})
     Y=0
     Z=&FloatFromCurve(FloatCurve=DataInterface() {propertyValues {Curve {keys [{arriveTangent=2.821, interpMode=RCIM_Cubic, leaveTangent=2.821, tangentMode=RCTM_User, time=0.59, value=0}, <at 1 0.5 RCIM_Cubic>]}}})
     W=0
    )
    Particles.InkBlotSize:Vector2f=&Multiply_Vector2D(
     A=$Particles.Initial.InkBlotSize
     B=&Vector2DFromFloat(Value=&FloatFromCurve(FloatCurve=DataInterface() {propertyValues {Curve {keys [<at 0 1>, {arriveTangent=1.2346, leaveTangent=1.2346, time=0.58, value=1}]}}}))
    )
   ) {isSetParameters=true}
  ]
  rendererClasses [&NiagaraMeshRendererProperties, &NiagaraSpriteRendererProperties]
  renderers [
   {rendererClass=&NiagaraMeshRendererProperties, rendererIndex=0}
   {rendererClass=&NiagaraSpriteRendererProperties, rendererIndex=1}
   {rendererClass=&NiagaraSpriteRendererProperties, rendererIndex=2}
  ]
 }
 Fire_Back {
  emitterUpdateScript [
   &EmitterState()
   &SpawnBurst_Instantaneous(Spawn Count=5, Use Spawn Probability=false, UseLoopCountLimit=true)
  ]
  particleSpawnScript [
   &PartitionParticles()
   &InitializeParticle(
    Color=&MakeLinearColorFromVectorAndFloat(
     "Vector (RGB)"=&MakeVectorFromLinearColorRGB(Color=&Multiply_LinearColorByFloat(
       Linear Color=<rgba 1 1 1 1>
       Float=&MakeCustomFloatFromBool(Bool=$User.IsFrontEnd, True Float=1, False Float=1)
      ))
     "Float (Alpha)"=1
    )
    Mesh Scale=&MakeVector(
     X=1
     Y=1
     Z=&Multiply_Float(
      A=&Lerp_Float(
       A=0.35
       B=0.6
       Alpha=&FloatFromCurve(
        FloatCurve=DataInterface() {propertyValues {Curve {keys [
            {interpMode=RCIM_Cubic, tangentMode=RCTM_User, time=0, value=0}
            {interpMode=RCIM_Cubic, tangentMode=RCTM_User, time=0.5, value=1}
            {tangentMode=RCTM_User, time=1, value=0}
           ]}}}
        CurveIndex=&ReturnNormalizedExecIndex()
       )
      )
      B=&MakeCustomFloatFromBool(Bool=$User.IsLargeBody, True Float=1.85, False Float=1.25)
     )
    )
    Mesh Scale Mode=::NewEnumerator4
    UsePositionOffset=false
    Write Material Random=true
   )
   &InitialMeshOrientation(Mesh Orientation Mode=::NewEnumerator1, Rotation=<0 0 0.3>, Use Rotation Vector=true)
   &CameraOffset()
  ]
  particleUpdateScript [
   &ParticleState(Kill Particles When Lifetime Has Elapsed=false, Let Infinitely Lived Particles Die When Emitter Deactivates=true)
   &DynamicMaterialParameters(
    Index 0 Param 1=&Multiply_Float(
     A=&Lerp_Float(
      A=-14
      B=-12
      Alpha=&FloatFromCurve(
       FloatCurve=DataInterface() {propertyValues {Curve {keys [
           {interpMode=RCIM_Cubic, tangentMode=RCTM_User, time=0, value=0}
           {interpMode=RCIM_Cubic, tangentMode=RCTM_User, time=0.5, value=1}
           {tangentMode=RCTM_User, time=1, value=0}
          ]}}}
       CurveIndex=&ReturnNormalizedExecIndex()
      )
     )
     B=&MakeCustomFloatFromBool(Bool=$User.IsLargeBody, True Float=1.6, False Float=1)
    )
    Index 3 Param 1=1
    Index 3 Param 2=1
    Index 3 Param 3=1
    Index 3 Param 4=1
    Module.Param0WriteEnabled=true
    Module.Param0WriteEnabled3=true
    Module.Param1WriteEnabled=true
    Module.Param1WriteEnabled3=true
    Module.Param2WriteEnabled=true
    Module.Param2WriteEnabled3=true
    Module.Param3WriteEnabled=true
    Module.Param3WriteEnabled3=true
    Write Parameter Index 3=true
   )
   &UpdateMeshOrientation(
    Facing Direction=/Game/Effects/Niagara/DynamicInputs/Engine/FN_GetSystemAxis(Axis:ENiagaraOrientationAxis=::XAxis, Flip Axis:NiagaraBool=false)
    Facing Interpolation=false
    Orientation Method=::NewEnumerator1
    Override Quat=false
    TransformFacing=false
    TransformUp=false
   ) {module=UpdateMeshOrientation001}
   &InterpolateOverTime(
    Rate Of Change:NiagaraFloat=&RandomRangeFloat(Minimum=5, Maximum=13)
    Target Vector Value:Vector3f=&ConvertPositionToVector(Input Position=$Particles.Position, Convert To Absolute Large World Vector=false)
    Variable Type=::NewEnumerator2
   )
   &SkeletalMeshLocation(
    Bone / Socket Index=&SelectIntFromArray(Direct Array Index=$Particles.PartitionParticles.Partition, Int Selection Array=DataInterface() {propertyValues {IntData [7]}})
    Bone Sampling Mode=::NewEnumerator4
    MaskSpawnGroup=false
    Orientation Sampling=::NewEnumerator1
    Sampled Position Offset=&Lerp_Vector(
     A=<-10 0 -12>
     B=&SelectVectorFromArray(Direct Array Index=$Particles.PartitionParticles.Partition, Vector Selection Array=DataInterface() {propertyValues {FloatData [<-10 0 12>]}})
     Alpha=&ReturnNormalizedExecIndex()
    )
    Skeletal Mesh=$System.Skeletal Mesh
   )
   /SetVariables(Particles.Interpolate Dynamic=&MakeFloat4FromVectorAndFloat(
     "Vector (RGB)"=$StackContext.InterpolateOverTime.InterpolateOverTime_V3_.Moving Average
     "Float (Alpha)"=&Add_Float(
      A=&MakeCustomFloatFromBool(Bool=$User.IsFrontEnd, True Float=2, False Float=$User.Interpolate Intensity)
      B=0.5
     )
    )) {isSetParameters=true, module=SetVariables_#2}
  ]
  rendererClasses [&NiagaraMeshRendererProperties]
  renderers [{rendererClass=&NiagaraMeshRendererProperties, rendererIndex=0}]
 }
 Fire_UpperArm {
  emitterUpdateScript [
   &EmitterState()
   &SpawnBurst_Instantaneous(Spawn Count=8, Use Spawn Probability=false, UseLoopCountLimit=true)
  ]
  particleSpawnScript [
   &PartitionParticles()
   &InitializeParticle(
    Color=&MakeLinearColorFromVectorAndFloat(
     "Vector (RGB)"=&MakeVectorFromLinearColorRGB(Color=&Multiply_LinearColorByFloat(
       Linear Color=<rgba 1 1 1 1>
       Float=&MakeCustomFloatFromBool(Bool=$User.IsFrontEnd, True Float=1, False Float=1)
      ))
     "Float (Alpha)"=1
    )
    Mesh Scale=&MakeVector(
     X=1
     Y=1
     Z=&Multiply_Float(
      A=&Lerp_Float(A=0.35, B=0.3, Alpha=&ReturnNormalizedExecIndex())
      B=&MakeCustomFloatFromBool(Bool=$User.IsLargeBody, True Float=1.85, False Float=1.25)
     )
    )
    Mesh Scale Mode=::NewEnumerator4
    UsePositionOffset=false
    Write Material Random=true
   )
   &InitialMeshOrientation(Mesh Orientation Mode=::NewEnumerator1, Rotation=<0 0 0.3>, Use Rotation Vector=true)
  ]
  particleUpdateScript [
   &ParticleState(Kill Particles When Lifetime Has Elapsed=false, Let Infinitely Lived Particles Die When Emitter Deactivates=true)
   &DynamicMaterialParameters(
    Index 0 Param 1=&Multiply_Float(
     A=&Lerp_Float(A=-14, B=-12, Alpha=&ReturnNormalizedExecIndex())
     B=&MakeCustomFloatFromBool(Bool=$User.IsLargeBody, True Float=1.6, False Float=1)
    )
    Index 3 Param 1=1
    Index 3 Param 2=1
    Index 3 Param 3=1
    Index 3 Param 4=1
    Module.Param0WriteEnabled=true
    Module.Param0WriteEnabled3=true
    Module.Param1WriteEnabled=true
    Module.Param1WriteEnabled3=true
    Module.Param2WriteEnabled=true
    Module.Param2WriteEnabled3=true
    Module.Param3WriteEnabled=true
    Module.Param3WriteEnabled3=true
    Write Parameter Index 3=true
   )
   &UpdateMeshOrientation(
    Facing Direction=&FN_GetSystemAxis(Axis:ENiagaraOrientationAxis=::XAxis, Flip Axis:NiagaraBool=false)
    Facing Interpolation=false
    Orientation Method=::NewEnumerator1
    Override Quat=false
    TransformFacing=false
    TransformUp=false
   ) {module=UpdateMeshOrientation001}
   &InterpolateOverTime(
    Rate Of Change:NiagaraFloat=&RandomRangeFloat(Minimum=5, Maximum=13)
    Target Vector Value:Vector3f=&ConvertPositionToVector(Input Position=$Particles.Position, Convert To Absolute Large World Vector=false)
    Variable Type=::NewEnumerator2
   )
   &SkeletalMeshLocation(
    Bone / Socket Index=&SelectIntFromArray(Direct Array Index=$Particles.PartitionParticles.Partition, Int Selection Array=DataInterface() {propertyValues {IntData [1,2]}})
    Bone Sampling Mode=::NewEnumerator4
    MaskSpawnGroup=false
    Orientation Sampling=::NewEnumerator1
    Sampled Position Offset=&Lerp_Vector(
     A=<0 0 0>
     B=&SelectVectorFromArray(
      Direct Array Index=$Particles.PartitionParticles.Partition
      Vector Selection Array=DataInterface() {propertyValues {FloatData [<28 0 0>, <-28 0 0>]}}
     )
     Alpha=&ReturnNormalizedExecIndex()
    )
    Skeletal Mesh=$System.Skeletal Mesh
   )
   /SetVariables(Particles.Interpolate Dynamic=&MakeFloat4FromVectorAndFloat(
     "Vector (RGB)"=$StackContext.InterpolateOverTime.InterpolateOverTime_V3_.Moving Average
     "Float (Alpha)"=&Add_Float(
      A=&MakeCustomFloatFromBool(Bool=$User.IsFrontEnd, True Float=2, False Float=$User.Interpolate Intensity)
      B=0.5
     )
    )) {isSetParameters=true, module=SetVariables_#2}
  ]
  rendererClasses [&NiagaraMeshRendererProperties]
  renderers [{rendererClass=&NiagaraMeshRendererProperties, rendererIndex=0}]
 }
 Glow {
  emitterUpdateScript [
   &EmitterState(
    Inactive Response=::NewEnumerator1
    Life Cycle Mode=::NewEnumerator1
    Loop Behavior=::NewEnumerator1
    Loop Duration Mode=::NewEnumerator1
    UseLoopDelay=false
   )
   &SpawnBurst_Instantaneous(Spawn Count=7, Use Spawn Probability=false, UseLoopCountLimit=false)
  ]
  particleSpawnScript [
   &PartitionParticles(Number Of Partitions=7)
   &InitializeParticle(
    Lifetime=3
    Sprite Rotation Mode=::NewEnumerator0
    Sprite Size Mode=::NewEnumerator3
    Uniform Sprite Size=&Multiply_Float(
     A=&SelectFloatFromArray(
      Direct Array Index=$Particles.PartitionParticles.Partition
      Float Selection Array=DataInterface() {propertyValues {FloatData [35,35,35,35,35,35]}}
     )
     B=&MakeCustomFloatFromBool(Bool=$User.IsLargeBody, True Float=1.5, False Float=1)
    )
    UsePositionOffset=false
    Write Material Random=true
   ) {module=InitializeParticle001}
   /SetVariables() {isSetParameters=true}
   /SetVariables(Particles.glowScale:Vector2f=&Vector2DFromFloat(Value=&Multiply_Float(
      A=&SelectFloatFromArray(
       Direct Array Index=$Particles.PartitionParticles.Partition
       Float Selection Array=DataInterface() {propertyValues {FloatData [150,100,100,65,65,50,50]}}
      )
      B=&MakeCustomFloatFromBool(Bool=$User.IsLargeBody, True Float=1.2, False Float=1)
     ))) {isSetParameters=true}
   /SetVariables(Particles.glowCameraOffset=-10) {isSetParameters=true}
   /SetVariables(Particles.cracksDecalScale:Vector2f=&Vector2DFromFloat(Value=&SelectFloatFromArray(
      Array Sampling Mode=::NewEnumerator0
      Direct Array Index=?
      Float Selection Array=DataInterface() {propertyValues {FloatData [30,15,15,30,30,30]}}
      Random Seed=0
     ))) {isSetParameters=true}
  ]
  particleUpdateScript [
   &ParticleState(
    Kill Particles When Lifetime Has Elapsed=false
    Let Infinitely Lived Particles Die When Emitter Deactivates=true
    Loop Particles Lifetime=true
   )
   &CameraOffset(Camera Offset Amount=5)
   &SkeletalMeshLocation(
    Bone / Socket Index=&SelectIntFromArray(Direct Array Index=$Particles.PartitionParticles.Partition, Int Selection Array=DataInterface() {propertyValues {IntData [0,1,2,3,4,5,6]}})
    Bone Sampling Mode=::NewEnumerator4
    MaskSpawnGroup=false
    Orientation Sampling=::NewEnumerator1
    Sampled Position Offset Coordinate Space:ENiagaraSkelMeshTransforms=::NewEnumerator1
    Skeletal Mesh=$System.Skeletal Mesh
   ) {module=SkeletalMeshLocation003}
   &DynamicMaterialParameters(
    Index 0 Param 1=&SelectFloatFromArray(Direct Array Index=$Particles.PartitionParticles.Partition, Float Selection Array=DataInterface() {propertyValues {FloatData [4,4,4,4,4,4]}})
    Index 0 Param 2=0.15
    Index 0 Param 3=1
    Module.Param0WriteEnabled=true
    Module.Param1WriteEnabled=true
    Module.Param2WriteEnabled=true
    Module.Param3WriteEnabled=true
   )
   &ScaleColor(ScaleA=true, ScaleRGB=true)
  ]
  rendererClasses [&NiagaraSpriteRendererProperties]
  renderers [{rendererClass=&NiagaraSpriteRendererProperties}]
 }
 Smoke {
  emitterUpdateScript [
   &EmitterState()
   &SpawnRate(SpawnRate=8, Use Spawn Probability=false)
  ]
  particleSpawnScript [
   &PartitionParticles(Number Of Partitions=6)
   &InitializeParticle(
    Color=&MakeLinearColorFromVectorAndFloat(
     "Vector (RGB)"=&MakeVectorFromLinearColorRGB(Color=&Multiply_LinearColorByFloat(
       Linear Color=&RandomRangeLinearColor(Minimum=<rgba 0.5885 0.5885 0.5885 0.35>, Maximum=<rgba 1 1 1 0.5>)
       Float=&MakeCustomFloatFromBool(Bool=$User.IsFrontEnd, True Float=1, False Float=1.3)
      ))
     "Float (Alpha)"=1
    )
    Lifetime Max=1.25
    Lifetime Min=0.5
    Lifetime Mode=::NewEnumerator1
    Sprite Rotation Mode=::NewEnumerator2
    Sprite Size Mode=::NewEnumerator1
    Uniform Sprite Size Max=40
    Uniform Sprite Size Min=20
    UsePositionOffset=false
    Write Material Random=true
   ) {module=InitializeParticle001}
   /SetVariables(Particles.RandomFrame=&RandomRangeInt(Minimum=0, Maximum=63)) {isSetParameters=true}
   &DynamicMaterialParameters(
    Index 0 Param 1=&RandomRangeFloat(Minimum=0.2, Maximum=0.5)
    Module.Param0WriteEnabled=true
    Module.Param1WriteEnabled=true
    Module.Param2WriteEnabled=true
    Module.Param3WriteEnabled=true
   )
   &SkeletalMeshLocation(
    Bone / Socket Index=&SelectIntFromArray(Direct Array Index=$Particles.PartitionParticles.Partition, Int Selection Array=DataInterface() {propertyValues {IntData [1,1,1,2,2,2]}})
    Bone Sampling Mode=::NewEnumerator4
    Inherit Velocity=false
    Interpolate Mesh Sampling=false
    MaskSpawnGroup=false
    Sampled Position Offset=&SelectVectorFromArray(
     Direct Array Index=$Particles.PartitionParticles.Partition
     Vector Selection Array=DataInterface() {propertyValues {FloatData [<0 0 0>, <12.5 0 0>, <25 0 0>, <0 0 0>, <-12.5 0 0>, <-25 0 0>]}}
    )
    Sampled Position Offset Coordinate Space:ENiagaraMeshTransforms=::NewEnumerator3
    Skeletal Mesh=$System.Skeletal Mesh
   )
   &AddVelocity(Velocity=&RandomRangeVector(Minimum=<0 0 40>, Maximum=<0 0 60>))
  ]
  particleUpdateScript [
   &ParticleState()
   &Drag(Drag=&RandomRangeFloat(Minimum=7, Maximum=10), Use Linear Drag=true, Use Rotational Drag=false)
   &GravityForce(Gravity=&RandomRangeVector(Minimum=<0 0 100>, Maximum=<0 0 150>)) {module=GravityForce002}
   &SolveForcesAndVelocity(Clamp Velocity=false, Limit Acceleration=false)
   &ScaleSpriteSize(Scale Factor=&Vector2DFromFloat(Value=&FloatFromCurve(FloatCurve=DataInterface() {propertyValues {Curve {keys [
          {arriveTangent=2.5076, interpMode=RCIM_Cubic, leaveTangent=2.5076, tangentMode=RCTM_User, time=0.0023, value=0}
          {arriveTangent=0.0897, interpMode=RCIM_Cubic, leaveTangent=0.0897, tangentMode=RCTM_User, time=1, value=1}
         ]}}})))
   &ScaleColor(
    Scale Alpha=&FloatFromCurve(
     FloatCurve=DataInterface() {propertyValues {Curve {keys [
         {arriveTangent=9.0725, interpMode=RCIM_Cubic, leaveTangent=9.0725, tangentMode=RCTM_User, time=0, value=0}
         {arriveTangent=-1.5704, interpMode=RCIM_Cubic, leaveTangent=-1.5704, tangentMode=RCTM_User, time=0.1408, value=0.9847}
         {arriveTangent=-0.016, interpMode=RCIM_Cubic, leaveTangent=-0.016, tangentMode=RCTM_User, time=1, value=0}
        ]}}}
     Scale Curve=&RandomRangeFloat(Minimum=0.5, Maximum=1)
    )
    ScaleA=true
    ScaleRGB=true
   ) {module=ScaleColor001}
   &DynamicMaterialParameters(
    Index 0 Param 1=&FloatFromCurve(
     FloatCurve=DataInterface() {propertyValues {Curve {keys [{arriveTangent=1.7573, interpMode=RCIM_Cubic, leaveTangent=1.7573, tangentMode=RCTM_User, time=0, value=0}, <at 1 1 RCIM_Cubic>]}}}
     Scale Curve=0.75
    )
    Module.Index0Param0Write=true
    Module.Index0Param1Write=true
    Module.Index0Param2Write=true
    Module.Index0Param3Write=true
   ) {module=DynamicMaterialParameters001}
   &SubUVAnimation(
    Random Start Frame=true
    "Sprite Renderer "=DataInterface() {propertyValues {SpriteRenderer=&self:Smoke_3.NiagaraSpriteRendererProperties_0}}
    SubUV Animation Mode:ENiagara_SubUVLookupModeV2=::NewEnumerator3
    UseEndFrame=false
    UsePlayRateScale=false
    UseStartFrame=false
   )
  ]
  rendererClasses [&NiagaraSpriteRendererProperties]
  renderers [{rendererClass=&NiagaraSpriteRendererProperties}]
 }
 Splotches {
  emitterUpdateScript [
   &EmitterState()
   &SpawnRate(SpawnRate=3.5, Use Spawn Probability=false)
  ]
  particleSpawnScript [
   &PartitionParticles(Number Of Partitions=5, Particles Per Partition=1, Partition Mode:ENiagara_PartitionMode=::NewEnumerator1)
   &InitializeParticle(
    Color=&MakeLinearColorFromVectorAndFloat(
     "Vector (RGB)"=&MakeVectorFromLinearColorRGB(Color=&Multiply_LinearColorByFloat(
       Linear Color=&RandomRangeLinearColor(Minimum=<rgba 0.0208 0.0208 0.0208 1>, Maximum=<rgba 0.0781 0.0781 0.0781 1>)
       Float=&MakeCustomFloatFromBool(Bool=$User.IsFrontEnd, True Float=1, False Float=1.4)
      ))
     "Float (Alpha)"=1
    )
    Lifetime=&RandomRangeFloat(Minimum=2.5, Maximum=3)
    UsePositionOffset=false
    Write Material Random=true
   )
   /SetVariables(Particles.PivotOffset:Vector2f=&Multiply_Vector2D_ByFloat(
     Float=&MakeCustomFloatFromBool(Bool=$User.IsLargeBody, True Float=1, False Float=1)
     Vector2D=&SelectVector2DFromArray(
      Direct Array Index=$Particles.PartitionParticles.Partition
      Direct Array Mode=::NewEnumerator1
      Vector 2D Selection Array=DataInterface() {propertyValues {FloatData [<0.2379 0.3>, <0.3621 0.7>, <0.6884 0.4>, <0.6473 0.6>, <0.5113 0.35>]}}
     )
    )) {isSetParameters=true}
   &SkeletalMeshLocation(Bone / Socket Index=8, Bone Sampling Mode=::NewEnumerator4, MaskSpawnGroup=false, Skeletal Mesh=$System.Skeletal Mesh) {module=SkeletalMeshLocation001}
   &ApplyInitialForces()
   /SetVariables(Particles.FrameRotation:NiagaraFloat=&RandomRangeFloat(Minimum=0, Maximum=1)) {isSetParameters=true}
   &DynamicMaterialParameters(
    Index 1 Param 1=&SelectFloatFromArray(
     Array Sampling Mode=::NewEnumerator0
     Direct Array Index=?
     Float Selection Array=DataInterface() {propertyValues {FloatData [0,1,2,3,4,5,6,7,8]}}
     Random Seed=0
    )
    Index 1 Param 2=1
    Index 1 Param 3=$Particles.FrameRotation
    Module.Index1Param0Write=true
    Module.Index1Param1Write=true
    Module.Index1Param2Write=true
    Module.Index1Param3Write=false
    Write Parameter Index 0=false
    Write Parameter Index 1=true
   ) {module=DynamicMaterialParameters001}
   /SetVariables(Particles.SplotchSizeSoftness:NiagaraFloat=&RandomRangeFloat(Minimum=0, Maximum=1)) {isSetParameters=true}
   /SetVariables(Particles.SpriteSize=&Multiply_Vector2D_ByFloat(
     Float=&MakeCustomFloatFromBool(Bool=$User.IsLargeBody, True Float=1.15, False Float=1)
     Vector2D=&Multiply_Vector2D(
      A=&Vector2DFromFloat(Value=&Lerp_Float(
        A=60
        B=90
        Alpha=&FloatFromCurve(
         FloatCurve=DataInterface() {propertyValues {Curve {keys [
             {arriveTangent=1.5, interpMode=RCIM_Cubic, time=0, value=0}
             {arriveTangent=0.8889, interpMode=RCIM_Cubic, leaveTangent=0.8889, time=0.5, value=0.2}
             {arriveTangent=1.6, interpMode=RCIM_Cubic, leaveTangent=1.6, time=0.9, value=0.8}
             <at 1 1 RCIM_Cubic>
            ]}}}
         CurveIndex=$Particles.SplotchSizeSoftness
        )
       ))
      B=&RandomRangeVector2D(Minimum=<0.8 0.8>, Maximum=<1.2 1.2>)
     )
    )) {isSetParameters=true}
  ]
  particleUpdateScript [
   &ParticleState()
   &CameraOffset(Camera Offset Amount=&Multiply_Float(A=-30, B=&MakeCustomFloatFromBool(Bool=$User.IsLargeBody, True Float=1.5, False Float=1)))
   &SkeletalMeshLocation(
    Bone / Socket Index=8
    Bone Sampling Mode=::NewEnumerator4
    MaskSpawnGroup=false
    Position Sampling=::NewEnumerator2
    Skeletal Mesh=$System.Skeletal Mesh
   )
   &SolveForcesAndVelocity(Clamp Velocity=false, Limit Acceleration=false)
   /SetVariables(Particles.FrameRotation:NiagaraFloat=&Add_Float(
     A=$Particles.Initial.FrameRotation
     B=&Multiply_Float(
      A=&FloatFromCurve(FloatCurve=DataInterface() {propertyValues {Curve {keys [{arriveTangent=2.514, interpMode=RCIM_Cubic, time=0, value=0}, {interpMode=RCIM_Cubic, leaveTangent=0.0224, time=1, value=0.0375}]}}})
      B=&SelectFloatFromArray(
       Array Sampling Mode=::NewEnumerator0
       Direct Array Index=?
       Float Selection Array=DataInterface() {propertyValues {FloatData [-1,1]}}
       Random Seed=0
      )
     )
    )) {isSetParameters=true}
   &DynamicMaterialParameters(
    Index 0 Param 1=&FloatFromCurve(
     FloatCurve=DataInterface() {propertyValues {Curve {keys [
         {arriveTangent=1.6306, interpMode=RCIM_Cubic, leaveTangent=1.6306, tangentMode=RCTM_User, time=0, value=0}
         <at 0.7 0.95 RCIM_Cubic>
         {arriveTangent=-6.1305, interpMode=RCIM_Cubic, leaveTangent=-6.1305, tangentMode=RCTM_User, time=1, value=0}
        ]}}}
     Scale Curve=&RandomRangeFloat(Minimum=0.9, Maximum=1)
    )
    Index 0 Param 2=&FloatFromCurve(FloatCurve=DataInterface() {propertyValues {Curve {keys [
         {arriveTangent=2.4446, interpMode=RCIM_Cubic, leaveTangent=2.4446, tangentMode=RCTM_User, time=0.4, value=-0.1}
         {arriveTangent=0.0442, interpMode=RCIM_Cubic, leaveTangent=0.0442, tangentMode=RCTM_User, time=1, value=1}
        ]}}})
    Index 0 Param 3=&FloatFromCurve(FloatCurve=DataInterface() {propertyValues {Curve {keys [
         {arriveTangent=1.6306, interpMode=RCIM_Cubic, leaveTangent=1.6306, tangentMode=RCTM_User, time=0, value=0}
         <at 0.7 0.5 RCIM_Cubic>
         {arriveTangent=-4.2528, interpMode=RCIM_Cubic, leaveTangent=-4.2528, tangentMode=RCTM_User, time=1, value=0}
        ]}}})
    Index 0 Param 4=&Lerp_Float(A=0.05, B=0.1, Alpha=$Particles.SplotchSizeSoftness)
    Index 1 Param 1=&SelectFloatFromArray(
     Array Sampling Mode=::NewEnumerator0
     Direct Array Index=?
     Float Selection Array=DataInterface() {propertyValues {FloatData [0,1,2,3]}}
     Random Seed=0
    )
    Index 1 Param 2=&FloatFromCurve(FloatCurve=DataInterface() {propertyValues {Curve {keys [
         {arriveTangent=5.9025, interpMode=RCIM_Cubic, leaveTangent=5.9025, tangentMode=RCTM_User, time=0, value=0}
         {interpMode=RCIM_Cubic, leaveTangent=1.8006, time=0.5, value=1}
        ]}}})
    Index 1 Param 3=$Particles.FrameRotation
    Module.Index0Param0Write=true
    Module.Index0Param1Write=false
    Module.Index0Param2Write=true
    Module.Index0Param3Write=true
    Module.Index1Param0Write=false
    Module.Index1Param1Write=false
    Module.Index1Param2Write=true
    Module.Index1Param3Write=true
    Write Parameter Index 1=true
   )
   /SetVariables(Particles.PivotOffset:Vector2f=&Lerp_Vector2(
     A=$Particles.Initial.PivotOffset
     B=<0.5 0.5>
     Alpha=&FloatFromCurve(
      FloatCurve=DataInterface() {propertyValues {Curve {keys [{tangentMode=RCTM_User, time=0, value=1}, {tangentMode=RCTM_User, time=1, value=0}]}}}
      Scale Curve=&RandomRangeFloat(Minimum=0.05, Maximum=0.3)
     )
    )) {isSetParameters=true}
   &InterpolateOverTime(
    Rate Of Change:NiagaraFloat=&ClampFloat(
     Float=&Add_Float(
      A=1
      B=&Multiply_Float(A=&VectorLength(Vector=$Particles.SkeletalMeshLocation.SampledVelocity), B=0.0225)
     )
     Min=1
     Max=20
    )
    Target Position Value:NiagaraPosition=$Particles.SkeletalMeshLocation.SampledPosition
    Variable Type=::NewEnumerator6
   )
   /SetVariables(Particles.Position=$Particles.InterpolateOverTime.InterpolateOverTime_Position.Moving Average) {isSetParameters=true}
  ]
  rendererClasses [&NiagaraSpriteRendererProperties]
  renderers [{rendererClass=&NiagaraSpriteRendererProperties}]
 }
}
summary {
 systemName=NS_Aura_Ink
 emitters [
  {emitterName=Glow, rendererClasses [&NiagaraSpriteRendererProperties]}
  {emitterName=Fire_UpperArm, rendererClasses [&NiagaraMeshRendererProperties]}
  {emitterName=Fire_Back, rendererClasses [&NiagaraMeshRendererProperties]}
  {emitterName=Smoke, rendererClasses [&NiagaraSpriteRendererProperties]}
  {emitterName=Butterflies, rendererClasses [&NiagaraMeshRendererProperties, &NiagaraSpriteRendererProperties]}
  {emitterName=Splotches, rendererClasses [&NiagaraSpriteRendererProperties]}
 ]
 userVariables [
  User.Interpolate Intensity {type=NiagaraFloat, defaultValue=0.35, description=""}
  User.IsFrontEnd {type=NiagaraBool, defaultValue=true}
  User.IsLargeBody {type=NiagaraBool}
  User.Spawn Sphere Radius {type=NiagaraFloat, defaultValue=200, description=""}
 ]
}
